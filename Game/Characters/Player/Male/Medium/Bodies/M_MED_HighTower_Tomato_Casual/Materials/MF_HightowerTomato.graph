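# Material function for the HighTower Tomato character's suit transition, revealing or dissolving body materials through an animated spherical mask.
# A curve atlas row drives two mask radii, hardness and gradient-noise distortion of pre-skinned position, with separate radius for the head.
# The masks, broken up by hashed noise, set opacity, inner color and edge-line emissive glow.
# The casual body flag inverts the opacity mask and keeps the original base color.
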MaterialFunction /Game/Characters/Player/Male/Medium/Bodies/M_MED_HighTower_Tomato_Casual/Materials/MF_HightowerTomato
graph {nodes [
  Constant() {r=1}
  n1 = GetMaterialAttributes(None=FunctionInput() {inputName=Material, inputType=FunctionInput_MaterialAttributes}) {attributeGetTypes [#1, #2, #3]}
  n2 = FunctionInput(Preview=VertexInterpolator(VS=PreSkinnedPosition()).PS) {bUsePreviewValueAsDefault=true, inputName=PreskinnedPosition}
  n3 = Divide(A=@n2, B=ScalarParameter() {defaultValue=1, parameterName=QuantizedNoiseCellDivision})
  n4 = ComponentMask(None=MaterialFunctionCall(Seed=Time(), Vec3=Ceil(None=@n3)) {materialFunction=&MF_Hash}.Vec3) {r=true}
  n5 = Clamp(None=OneMinus(None=Power(Base=@n4)))
  n6 = ScalarParameter() {defaultValue=1, parameterName=QuantizedNoiseStrength}
  n7 = Add(
   A=Divide(A=@n2, B=ScalarParameter() {defaultValue=32, parameterName=NoiseDivision})
   B=Multiply(A=Time(), B=VectorParameter() {parameterName=NoisePanSpeedAndDirection, defaultValue=<rgba 0.15 0.25 -0.1 1>}.RGB)
  )
  n8 = CurveAtlasRowParameter(CurveTime=ScalarParameter() {parameterName=HT_L1_TransitionCurve}) {
   atlas=&AthenaCurveAtlas01
   curve=&/HighTowerTomato_SuitTransitionCurve
   defaultValue=236
   parameterName=HT_TransitionCurve
  }
  n9 = LinearInterpolate(
   A=ScalarParameter() {defaultValue=12, parameterName=CurlNoise_EndStrength}
   B=ScalarParameter() {parameterName=CurlNoise_StartStrength}
   Alpha=@n8.A
  )
  n10 = Add(A=@n2, B=Multiply(A=Noise!(World Position=@n7, FilterWidth) {levels=5, noiseFunction=NOISEFUNCTION_GradientALU}, B=@n9))
  n11 = VectorParameter() {parameterName=MaskCenter, defaultValue=<rgba 0 8.5591 128 1>}
  n12 = ScalarParameter() {parameterName=BisHead}
  n13 = LinearInterpolate(
   A=ScalarParameter() {defaultValue=128, parameterName=MaskRadius}
   B=ScalarParameter() {defaultValue=24, parameterName=MaskRadius_Head}
   Alpha=@n12
  )
  n14 = Multiply(A=@n13, B=LinearInterpolate(A=@n8.R, B=@n8.R, Alpha=@n12))
  n15 = LinearInterpolate(
   A=ScalarParameter() {defaultValue=0.5, parameterName=MaskBaseHardness_Start, sliderMin=0.5}
   B=ScalarParameter() {defaultValue=1, parameterName=MaskBaseHardness_End}
   Alpha=@n8.A
  )
  n16 = Subtract(A=Multiply(A=SphereMask(A=@n10, B=@n11.RGB, Radius=@n14, Hardness=@n15)) {constB=2})
  n17 = TextureSampleParameter2D() {
   parameterName=EdgeMask
   samplerType=SAMPLERTYPE_LinearColor
   texture=/Game/Characters/Player/Male/Medium/Bodies/M_MED_HighTower_Tomato_Casual/Textures/T_HighTowerTomato_EdgeLines
  }
  n18 = Time()
  n19 = Divide(A=@n2, B=ScalarParameter() {defaultValue=1, parameterName=NoiseCellDivision})
  n20 = ConstantBiasScale(None=ComponentMask(None=MaterialFunctionCall(Seed=@n18, Vec3=@n19) {materialFunction=&MF_Hash}.Vec3) {r=true})
  n21 = ScalarParameter() {defaultValue=0.5, parameterName=HT_EdgeLineMaskingStrength}
  n22 = Add(
   A=Subtract(A=@n16, B=Multiply(A=@n5, B=@n6))
   B=Subtract(A=ChannelMaskParameter(None=@n17.RGB) {parameterName=HT_EdgelinesChannel}.R, B=Multiply(A=@n20, B=@n21) {constB=0.5})
  )
  n23 = Multiply(
   A=Multiply(A=VectorParameter() {parameterName=EdgeGlowColor, defaultValue=<rgba 0 0.3233 1 1>}.RGB, B=Clamp(None=@n22))
   B=ScalarParameter() {defaultValue=2, parameterName=EdgeGlowStrength}
  )
  n24 = Multiply(A=@n13, B=LinearInterpolate(A=@n8.G, B=@n8.G, Alpha=@n12))
  n25 = Subtract(A=Multiply(A=SphereMask(A=@n10, B=@n11.RGB, Radius=@n24, Hardness=@n15)) {constB=2})
  n26 = ScalarParameter() {defaultValue=15, parameterName=MaskBaseFalloff}
  n27 = Multiply(A=@n20, B=ScalarParameter() {defaultValue=15, parameterName=MaskNoise_Strength})
  n28 = Clamp(None=Subtract(A=Multiply(A=@n25, B=@n26) {constB=512}, B=@n27))
  n29 = ScalarParameter() {defaultValue=15, parameterName=MaskBaseFalloff}
  n30 = Clamp(None=Subtract(A=Multiply(A=@n16, B=@n29) {constB=512}, B=@n27))
  n31 = ScalarParameter() {parameterName=bIsCasualBody}
  n32 = LinearInterpolate(A=@n30, B=LinearInterpolate(Alpha=@n30) {constA=1, constB=0}, Alpha=@n31)
  n33 = LinearInterpolate(A=VectorParameter() {parameterName=InnerColor, defaultValue=<rgba 0.1218 0.1361 0.1406 1>}.RGB, B=@n1.BaseColor, Alpha=@n28)
  n34 = SetMaterialAttributes!(
   MaterialAttributes=@n1.MaterialAttributes
   BaseColor=LinearInterpolate(A=@n33, B=@n1.BaseColor, Alpha=@n31)
   EmissiveColor=Add(A=@n1.EmissiveColor, B=LinearInterpolate(A=@n23, B=@n1.EmissiveColor, Alpha=@n28) {constB=0})
   OpacityMask=@n32
   Opacity=Multiply(A=@n1.Opacity, B=@n32)
  ) {attributeSetTypes [
    #1
    #2
    #4
    #3
   ]}
  FunctionOutput(None=@n34)
  LinearInterpolate(A=@n9, Alpha=@n12) {constB=0}
  n35 = Constant() {r=0.5}
  n36 = ScalarParameter() {defaultValue=0.75, parameterName=VertexOffsetEdge_Radius, sliderMin=5}
  n37 = ScalarParameter() {defaultValue=0.25, parameterName=VertexOffset_EdgeHardness}
  LinearInterpolate(
   A=@n33
   B=VectorParameter() {parameterName=LeadingEdgeColor, defaultValue=<rgba 0 0 0 1>}.RGB
   Alpha=Clamp(None=SphereMask(A=@n16, B=@n35, Radius=@n36, Hardness=@n37)) {maxDefault=0}
  )
  ObjectRadius()
  ScalarParameter() {parameterName=HT_L2_TransitionCurve}
  VectorNoise(World Position=@n7) {noiseFunction=VNF_CurlALU}
  VertexInterpolator(VS=@n2)
  VertexInterpolator(VS=PreSkinnedPosition())
 ]}
