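# Master material for the HighTower Tomato Casual male player body, masked and built on material attributes for skeletal meshes.
# Shared character PreFX output runs through emissive fresnel, then static-switched layers for HighTower Tomato tech, silk, clear coat, metals LUT and holo foil, before shared PostFX.
# SkinFX_Mask drives the silk, clear coat and metal layers.
# A final world position offset lerps vertices toward the object position by HT_CrunchVerts, allowing the body to collapse.
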Material /Game/Characters/Player/Male/Medium/Bodies/M_MED_HighTower_Tomato_Casual/Materials/M_MED_HighTowerTomato_MASTER
graph {
 properties {
  bAutomaticallySetUsageInEditor=false
  bUseMaterialAttributes=true
  bUsedWithMeshDeformer=true
  bUsedWithMorphTargets=true
  bUsedWithSkeletalMesh=true
  blendMode=BLEND_Masked
  pixelDepthOffsetMode=PDOM_Legacy
  shadingModel=MSM_FromMaterialExpression
  usedShadingModels=Default Lit | Clear Coat | Subsurface Profile
 }
 nodes [
  n1 = MaterialFunctionCall() {materialFunction=&BaseFN_CharacterFunction_PreFX}
  n2 = VertexInterpolator(VS=PreSkinnedNormal())
  n3 = TextureSampleParameter2D() {parameterName=SkinFX_Mask}
  n4 = MaterialFunctionCall!(In=@n1.Result) {materialFunction=/Game/Weapons/WeaponSkins/MaterialFunction/MF_EmissiveFresnel}
  n5 = StaticSwitchParameter(
   True=MaterialFunctionCall!(PreskinnedPosition=@n1.Pre skinned Local Position, Material=@n4.Result) {materialFunction=&/MF_HightowerTomato}.Result
   False=@n4.Result
  ) {parameterName=UseHT_SuiteTech}
  n6 = StaticSwitchParameter(
   True=MaterialFunctionCall!(Alpha=ChannelMaskParameter(None=@n3.RGB) {group=Skin FX, parameterName=Silk Channel}.R, Material=@n5) {materialFunction=/Game/Weapons/WeaponSkins/MaterialFunction/MF_Silk}.Result
   False=@n5
  ) {group=Skin FX, parameterName=Use Silk}
  n7 = StaticSwitchParameter(
   True=MaterialFunctionCall!(Alpha=StaticComponentMaskParameter(None=@n3.RGB) {defaultR=true, group=Skin FX, parameterName=Clear Coat Channel}, In=@n6) {materialFunction=/Game/Weapons/WeaponSkins/MaterialFunction/Clearcoat}.Result
   False=@n6
  ) {group=Skin FX, parameterName=Use Clear Coat}
  n8 = StaticSwitchParameter(
   True=MaterialFunctionCall(Mask=@n3.RGBA, Materials=@n7) {materialFunction=&MF_MetalsLUT}.Result
   False=@n7
  ) {group=Skin FX, parameterName=UseMetalLUT}
  n9 = MaterialFunctionCall(
   AO=@n1.AO
   Distance Blend=@n1.AO
   Fresnel=@n1.AO
   Gmap=@n1.AO
   Material=StaticSwitchParameter(
    True=MaterialFunctionCall!(PreskinnedLocalPosition=@n1.Pre skinned Local Position, PreskinnedLocalNormal=@n2.PS, Material=@n8) {materialFunction=/Game/Characters/Player/Female/Medium/Bodies/F_MED_HighTower_Squash/Materials/MF_HoloFoil}.Result
    False=@n8
   ) {parameterName=BUseHoloFoil}
  ) {materialFunction=&BaseFN_CharacterFunction_PostFX}
  n10 = GetMaterialAttributes(None=@n9.Result) {attributeGetTypes [#1]}
  n11 = Subtract(A=ObjectPositionWS(), B=Add(A=WorldPosition().XYZ, B=@n10.World Position Offset))
  n12 = ScalarParameter() {parameterName=HT_CrunchVerts}
  SetMaterialAttributes!(MaterialAttributes=@n10.MaterialAttributes, World Position Offset=Add(A=@n10.World Position Offset, B=Multiply(A=@n11, B=@n12))) {attributeSetTypes [#1]}
 ]
}
