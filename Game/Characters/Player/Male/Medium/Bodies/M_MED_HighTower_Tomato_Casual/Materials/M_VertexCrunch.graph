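# Unlit skeletal mesh material for the HighTower Tomato Casual male body that collapses vertices into the object's position, shrinking the mesh to a point.
# World Position Offset is object position minus world position, scaled by saturated HT_CrunchVerts, fully collapsed by default.
# The attributes, with a constant opacity mask, pass through BaseFN_CharacterOpacity, noted as possibly removable.
# A summed Hide Element parameter chain ends in a vertex-color mask switch that feeds nothing.
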Material /Game/Characters/Player/Male/Medium/Bodies/M_MED_HighTower_Tomato_Casual/Materials/M_VertexCrunch
graph {
 properties {
  bUseMaterialAttributes=true
  bUsedWithMeshDeformer=true
  bUsedWithMorphTargets=true
  bUsedWithSkeletalMesh=true
  pixelDepthOffsetMode=PDOM_Legacy
  shadingModel=MSM_Unlit
  usedShadingModels=Unlit
 }
 nodes [
  n1 = Subtract(A=ObjectPositionWS(), B=WorldPosition().XYZ)
  n2 = Saturate(None=ScalarParameter() {defaultValue=1, parameterName=HT_CrunchVerts})
  MaterialFunctionCall!(Material=MakeMaterialAttributes!(
    BaseColor
    Metallic
    Specular
    Roughness
    Anisotropy
    EmissiveColor
    Opacity
    OpacityMask=Constant()
    Normal
    Tangent
    WorldPositionOffset=Multiply(A=@n1, B=@n2)
    SubsurfaceColor
    ClearCoat
    ClearCoatRoughness
    AmbientOcclusion
    Refraction
    CustomizedUVs_0
    CustomizedUVs_1
    CustomizedUVs_2
    CustomizedUVs_3
    CustomizedUVs_4
    CustomizedUVs_5
    CustomizedUVs_6
    CustomizedUVs_7
    PixelDepthOffset
    ShadingModel
    Displacement
   )) {
   desc=Can Be Removed?
   materialFunction=/Game/Packages/Fortress_SharedMaterials/Base_Material_Functions/BaseCharacterShaderFunctions/BaseFN_CharacterOpacity
  }
  n3 = ScalarParameter() {defaultValue=1, group=HideElements, parameterName=Hide Element 04, sliderMax=1}
  n4 = ScalarParameter() {defaultValue=1, group=HideElements, parameterName=Hide Element 9_5, sliderMax=1}
  n5 = ScalarParameter() {defaultValue=1, group=HideElements, parameterName=Hide Element 7_5, sliderMax=1}
  n6 = ScalarParameter() {defaultValue=1, group=HideElements, parameterName=Hide Element 6_5, sliderMax=1}
  n7 = ScalarParameter() {defaultValue=1, group=HideElements, parameterName=Hide Element 00, sliderMax=1}
  n8 = ScalarParameter() {defaultValue=1, group=HideElements, parameterName=Hide Element 02, sliderMax=1}
  n9 = ScalarParameter() {defaultValue=1, group=HideElements, parameterName=Hide Element 08, sliderMax=1}
  n10 = ScalarParameter() {defaultValue=1, group=HideElements, parameterName=Hide Element 3_5, sliderMax=1}
  n11 = Add(A=Add(A=Add(A=@n3, B=@n4), B=Add(A=@n5, B=@n6)), B=Add(A=Add(A=@n7, B=@n8), B=Add(A=@n9, B=@n10)))
  n12 = Add(
   A=Add(
    A=ScalarParameter() {defaultValue=1, group=HideElements, parameterName=Hide Element 01, sliderMax=1}
    B=ScalarParameter() {defaultValue=1, group=HideElements, parameterName=Hide Element 03, sliderMax=1}
   )
   B=Add(
    A=ScalarParameter() {defaultValue=1, group=HideElements, parameterName=Hide Element 5_5, sliderMax=1}
    B=ScalarParameter() {defaultValue=1, group=HideElements, parameterName=Hide Element 1_5, sliderMax=1}
   )
  )
  n13 = Add(
   A=Add(
    A=ScalarParameter() {defaultValue=1, group=HideElements, parameterName=Hide Element 06, sliderMax=1}
    B=ScalarParameter() {defaultValue=1, group=HideElements, parameterName=Hide Element 09, sliderMax=1}
   )
   B=Add(
    A=ScalarParameter() {defaultValue=1, group=HideElements, parameterName=Hide Element 10, sliderMax=1}
    B=ScalarParameter() {defaultValue=1, group=HideElements, parameterName=Hide Element 4_5, sliderMax=1}
   )
  )
  n14 = Add(
   A=ScalarParameter() {defaultValue=1, group=HideElements, parameterName=Hide Element 07, sliderMax=1}
   B=ScalarParameter() {defaultValue=1, group=HideElements, parameterName=Hide Element 05, sliderMax=1}
  )
  n15 = Add(
   A=ScalarParameter() {defaultValue=1, group=HideElements, parameterName=Hide Element 0_5, sliderMax=1}
   B=ScalarParameter() {defaultValue=1, group=HideElements, parameterName=Hide Element 2_5, sliderMax=1}
  )
  n16 = Add(
   A=Add(A=Add(A=@n12, B=@n13), B=Add(A=@n14, B=@n15))
   B=ScalarParameter() {group=HideElements, parameterName=HideElementsOnlyConsidersRedChannel}
  )
  StaticSwitchParameter(True=LinearInterpolate(B=Add(A=@n11, B=@n16) {desc=Plugged into clamp to save an intstruction}) {constAlpha=0}) {defaultValue=true, group=Opacity Mask, parameterName=Use Vertex Colors for Mask}
 ]
}
